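; BP_YokeHopUnkindMutator is a mutator Blueprint whose name ties it to a YokeHopUnkind mode, with a switch for a prototype rectangle storm.
; The custom event EnableRectangleStormProto sets the ConsoleRectangleProtoEnabled variable to true; nothing in this file reads it.
; The InitializeSafeZones override only returns the parent result, while BeginPlay and the construction script are empty.
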(event Custom|EnableRectangleStormProto
  (Variables|Default|SetConsoleRectangleProtoEnabled true))

(event EventBeginPlay)

(fn InitializeSafeZones
  (bind _returnvalue (|CallParentFunction))
  (return _returnvalue))

(fn ConstructionScript)
